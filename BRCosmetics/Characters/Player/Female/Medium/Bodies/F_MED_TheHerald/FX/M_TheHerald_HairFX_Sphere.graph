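# Translucent, two-sided unlit material for Niagara mesh particles, used for The Herald's hair FX sphere.
# Emissive blends a dimmed, hue-shifted copy of a Fresnel two-color gradient toward the original using two combined panning noises.
# Opacity multiplies that noise by a facing-based edge fade and sine UV masks, with depth fade on default and deferred paths.
# Noise-driven world position offset displaces vertices along normals on those paths, and is zero on forward and mobile.
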Material /BRCosmetics/Characters/Player/Female/Medium/Bodies/F_MED_TheHerald/FX/M_TheHerald_HairFX_Sphere
graph {
 properties {
  bUsedWithNiagaraMeshParticles=true
  blendMode=BLEND_Translucent
  pixelDepthOffsetMode=PDOM_Legacy
  shadingModel=MSM_Unlit
  twoSided=true
  usedShadingModels=Unlit
 }
 nodes [
  n1 = LinearInterpolate(
   A=VectorParameter() {parameterName=ColorA, defaultValue=<rgba 0.4635 0.0269 0.1011 1>}.RGB
   B=VectorParameter() {parameterName=ColorB, defaultValue=<rgba 0.2083 0.0146 0.0564 1>}.RGB
   Alpha=Power(Base=Fresnel() {baseReflectFraction=0, exponent=1}, Exp=ScalarParameter() {defaultValue=0.3, parameterName=ColorLerpExp}) {constExponent=0.5}
  )
  n2 = MaterialFunctionCall(Hue Shift Percentage=ScalarParameter() {defaultValue=0.92, parameterName=HueShift}, Texture=@n1) {materialFunction=&HueShift}
  n3 = Constant() {r=3}
  n4 = VectorParameter() {parameterName=Tile, defaultValue=<rgba 1.5 0.2 0 0>}
  n5 = Multiply(A=TextureCoordinate(), B=AppendVector(A=@n4.R, B=@n4.R))
  n6 = Multiply(A=Time(), B=Constant() {r=1.5})
  n7 = TextureSample(UVs=Panner(Coordinate=@n5, Time=@n6) {speedX=-0.05, speedY=0.2}) {
   samplerType=SAMPLERTYPE_LinearGrayscale
   texture=/BRCosmetics/Athena/Effects/Particles/Skydiving/NS_Contrail_DarkAndLight/T_DarkAndLight_Noise
  }
  n8 = Add(
   A=@n7.R
   B=TextureSample(UVs=Panner(Coordinate=@n5, Time=@n6) {speedX=0.1, speedY=0.3}) {samplerType=SAMPLERTYPE_Grayscale, texture=&LowResBlurredNoise}.R
  )
  n9 = Constant() {r=5}
  MaterialFunctionCall(Emissive=LinearInterpolate(A=Divide(A=@n2.Result, B=@n3) {constB=1.5}, B=@n1, Alpha=Power(Base=@n8, Exp=@n9))) {materialFunction=&MF_FX_EmissiveAdjustments}
  n10 = CameraVectorWS()
  n11 = Multiply(A=VertexNormalWS())
  n12 = Power(Base=Abs(None=DotProduct(A=@n10, B=@n11)), Exp=Constant() {r=2})
  n13 = Saturate(None=Multiply(A=Saturate(None=@n12), B=Constant() {r=5}))
  n14 = MaterialFunctionCall(Float2=TextureCoordinate()) {materialFunction=&BreakOutFloat2Components}
  n15 = Subtract(A=Sine(None=Multiply(A=@n14.G) {constB=0.5})) {constB=0.1}
  n16 = Saturate(None=Multiply(A=@n15, B=@n15))
  n17 = Multiply(A=@n14.R) {constB=0.5}
  n18 = Saturate(None=Multiply(A=Sine(None=@n17)) {constB=4})
  n19 = Multiply(A=Multiply(A=Multiply(A=@n13, B=@n8), B=Multiply(A=@n16, B=@n18)), B=ScalarParameter() {defaultValue=1.5, parameterName=OverallOpacity})
  n20 = DepthFade(Opacity=@n19, FadeDistance=Constant() {r=2})
  ShadingPathSwitch(Default=@n20, Deferred=@n20, Forward=@n19, Mobile=@n19)
  n21 = Constant() {r=-0.8}
  n22 = TextureCoordinate()
  n23 = VectorParameter() {parameterName=WPO_NosieTile, defaultValue=<rgba 5 1 0 0>}
  n24 = AppendVector(A=@n23.R, B=@n23.R)
  n25 = TextureSample(UVs=Panner(Coordinate=Multiply(A=@n22, B=@n24)) {speedX=0.1, speedY=0.3}) {
   samplerType=SAMPLERTYPE_LinearGrayscale
   texture=&T_DarkAndLight_Noise
  }
  n26 = Multiply(A=Multiply(A=VertexNormalWS(), B=LinearInterpolate(A=@n21, Alpha=@n25.R) {constA=-1}), B=Constant() {r=2.4})
  n27 = Constant()
  ShadingPathSwitch(Default=@n26, Deferred=@n26, Forward=@n27, Mobile=@n27)
 ]
}
