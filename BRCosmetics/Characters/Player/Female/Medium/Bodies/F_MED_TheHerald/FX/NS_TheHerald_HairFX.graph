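# Niagara hair effect for The Herald character body, with default purple and dark gold versions selected by spawn probability from a system variant value.
# Mesh and sprite hair layers follow the skeletal mesh and fade with a speed lerp, alongside movement-spawned embers and gold-only ambient embers.
# Visibility tags respond to the head-hidden flag.
# A feather proto emitter runs a leashed attraction-and-wind simulation, exporting offset, body-relative rotation and sway to Blueprint at platform-dependent intervals.
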NiagaraSystem /BRCosmetics/Characters/Player/Female/Medium/Bodies/F_MED_TheHerald/FX/NS_TheHerald_HairFX
dependencies {
 usedDataInterfaces [
  &NiagaraDataInterfaceArrayFloat
  &NiagaraDataInterfaceArrayFloat3
  &NiagaraDataInterfaceCurve
  &NiagaraDataInterfacePlatformSet
 ]
 usedDynamicInputs [
  &MakeCustomFloatFromBool
  &RemapRange
  &Add_Vector
  &Boolean_AndOperation
  &SetBoolByFloatComparison
  &MakeFloatFromVector
  /Niagara/DynamicInputs/Transforms/TransformPosition
  &Boolean_OrOperation
 ]
 usedModules [
  &self:SystemSpawnScript.SystemScriptSource.SystemScriptGraph.NiagaraNodeAssignment_2.SetVariables_#1
  &self:SystemSpawnScript.SystemScriptSource.SystemScriptGraph.NiagaraNodeAssignment_3.SetVariables_#2
  &SystemState
  &self:SystemSpawnScript.SystemScriptSource.SystemScriptGraph.NiagaraNodeAssignment_1.SetVariables_#3
  &self:SystemSpawnScript.SystemScriptSource.SystemScriptGraph.NiagaraNodeAssignment_0.SetVariables_#4
  &DoOnce
  &InterpolateOverTime
 ]
}
emitters {
 Embers_DarkGold {
  emitterUpdateScript [
   &EmitterState()
   &SpawnPerUnit(
    Max Movement Threshold=2000
    Module.Use Max Movement Threshold=true
    Module.Use Movement Tolerance=false
    Spawn Probability:NiagaraFloat=&OneMinusFloat(Float=$System.Sys_Variant)
    Spawn Spacing=40
    Use Spawn Probability=true
   )
  ]
  particleSpawnScript [
   &InitializeParticle(
    Color=<rgba 1 1 1 0.7>
    Lifetime=&Lerp_Float(A=0.3, B=0.4, Alpha=$System.SpeedLerp)
    Sprite Rotation Mode=::NewEnumerator2
    Sprite Size Max=<15 15>
    Sprite Size Min=<8 8>
    Sprite Size Mode=::NewEnumerator2
    Sprite UV Mode=::NewEnumerator3
    UsePositionOffset=false
    Write Material Random=true
   )
   &SkeletalMeshLocation(Bone Sampling Mode=::NewEnumerator4, MaskSpawnGroup=false, Sampled Position Offset=<15 0 0>, Skeletal Mesh=$User.SK_Mesh)
   &ShapeLocation(MaskSpawnGroup=false, Override Seed=false, Sphere Radius=10, UseOwnerScale=true)
   &AddVelocity(Velocity=&MakeVector(X=0, Y=0, Z=&RandomRangeFloat(Minimum=20, Maximum=38)))
   &CameraOffset(Camera Offset Amount=-5)
  ]
  particleUpdateScript [
   &ParticleState()
   &InheritVelocity(Inherited Velocity Amount Scale=<0.1 0.1 0.1>, Inherited Velocity Speed Limit=1000, Speed Limit=true, UseSpeedThreshold=false)
   &InheritSourceMovement(Applied Position Scale=<0.85 0.85 0.9>, ApplyForce=false, ApplyPosition=true, ApplyPositionFalloff=false)
   &SolveForcesAndVelocity(Clamp Velocity=false, Limit Acceleration=false)
   &DynamicMaterialParameters(
    Index 0 Param 1=&FloatFromCurve(FloatCurve=DataInterface() {propertyValues {Curve {keys [{tangentMode=RCTM_User, time=0, value=0}, <at 0.2 1>, {tangentMode=RCTM_User, time=1, value=0}]}}})
    Module.Param0WriteEnabled=true
    Module.Param1WriteEnabled=true
    Module.Param2WriteEnabled=true
    Module.Param3WriteEnabled=true
   )
   /SetVariables(Particles.VisibilityTag=&MakeCustomIntFromBool(Bool=$User.bisHeadHidden, True INT=1, False INT=0)) {isSetParameters=true, module=SetVariables_#5}
  ]
  rendererClasses [&NiagaraSpriteRendererProperties]
  renderers [{rendererClass=&NiagaraSpriteRendererProperties}]
 }
 Embers_DarkGoldLocal {
  emitterUpdateScript [
   &EmitterState()
   &SpawnRate(
    Spawn Probability=&Multiply_Float(A=$System.SpeedLerp2, B=&OneMinusFloat(Float=$System.Sys_Variant))
    SpawnRate=3
    Use Spawn Probability=true
   )
  ]
  particleSpawnScript [
   &InitializeParticle(
    Color=<rgba 1 1 1 0.7>
    Lifetime Max=0.9
    Lifetime Min=0.6
    Lifetime Mode=::NewEnumerator1
    Sprite Size Max=<3 13>
    Sprite Size Min=<2 8>
    Sprite Size Mode=::NewEnumerator2
    UsePositionOffset=false
    Write Material Random=true
   )
   &TorusLocation(
    Align Torus With Axis=true
    Large Radius=120
    MaskSpawnGroup=false
    Module.Override Randomness=false
    Torus Axis=<0 1 1>
    Torus Mode=::NewEnumerator2
   )
   &AddVelocity(Velocity=$Output.TorusLocation.TorusNormal, Velocity Speed Scale=6)
   &CameraOffset(Camera Offset Amount=1)
   &SystemLocation(MaskSpawnGroup=false, Offset=<0 -5 0>)
   &AddVelocity(Velocity=<0 -3 5>) {module=AddVelocity001}
   &SkeletalMeshLocation(Bone Sampling Mode=::NewEnumerator4, MaskSpawnGroup=false, Skeletal Mesh=$User.SK_Mesh)
  ]
  particleUpdateScript [
   &ParticleState()
   &SolveForcesAndVelocity(Clamp Velocity=false, Limit Acceleration=false)
   &DynamicMaterialParameters(
    Index 0 Param 1=&FloatFromCurve(FloatCurve=DataInterface() {propertyValues {Curve {keys [{tangentMode=RCTM_User, time=0, value=0}, <at 0.2 1>, {tangentMode=RCTM_User, time=1, value=0}]}}})
    Module.Param0WriteEnabled=true
    Module.Param1WriteEnabled=true
    Module.Param2WriteEnabled=true
    Module.Param3WriteEnabled=true
   )
   &ScaleSpriteSize(Scale Sprite Size Mode=::NewEnumerator0, Uniform Scale Factor=1.3)
   /SetVariables(Particles.VisibilityTag=&MakeCustomIntFromBool(Bool=$User.bisHeadHidden, True INT=1, False INT=0)) {isSetParameters=true, module=SetVariables_#5}
  ]
  rendererClasses [&NiagaraSpriteRendererProperties]
  renderers [{rendererClass=&NiagaraSpriteRendererProperties}]
 }
 Embers_DefaultPurple {
  emitterUpdateScript [
   &EmitterState()
   &SpawnPerUnit(
    Max Movement Threshold=2000
    Module.Use Max Movement Threshold=true
    Module.Use Movement Tolerance=false
    Spawn Probability:NiagaraFloat=$System.Sys_Variant
    Spawn Spacing=40
    Use Spawn Probability=true
   )
  ]
  particleSpawnScript [
   &InitializeParticle(
    Color=<rgba 1 1 1 0.7>
    Lifetime=&Lerp_Float(A=0.3, B=0.4, Alpha=$System.SpeedLerp)
    Sprite Rotation Mode=::NewEnumerator2
    Sprite Size Mode=::NewEnumerator2
    Sprite UV Mode=::NewEnumerator3
    UsePositionOffset=false
    Write Material Random=true
   )
   &SkeletalMeshLocation(Bone Sampling Mode=::NewEnumerator4, MaskSpawnGroup=false, Sampled Position Offset=<15 0 0>, Skeletal Mesh=$User.SK_Mesh)
   &ShapeLocation(MaskSpawnGroup=false, Override Seed=false, Sphere Radius=10, UseOwnerScale=true)
   &AddVelocity(Velocity=&MakeVector(X=0, Y=0, Z=&RandomRangeFloat(Minimum=20, Maximum=38)))
   &CameraOffset(Camera Offset Amount=-5)
  ]
  particleUpdateScript [
   &ParticleState()
   &InheritVelocity(Inherited Velocity Amount Scale=<0.1 0.1 0.1>, Inherited Velocity Speed Limit=1000, Speed Limit=true, UseSpeedThreshold=false)
   &InheritSourceMovement(Applied Position Scale=<0.85 0.85 0.9>, ApplyForce=false, ApplyPosition=true, ApplyPositionFalloff=false)
   &SolveForcesAndVelocity(Clamp Velocity=false, Limit Acceleration=false)
   &DynamicMaterialParameters(
    Index 0 Param 1=&FloatFromCurve(FloatCurve=DataInterface() {propertyValues {Curve {keys [{tangentMode=RCTM_User, time=0, value=0}, <at 0.2 1>, {tangentMode=RCTM_User, time=1, value=0}]}}})
    Module.Param0WriteEnabled=true
    Module.Param1WriteEnabled=true
    Module.Param2WriteEnabled=true
    Module.Param3WriteEnabled=true
   )
   /SetVariables(Particles.VisibilityTag=&MakeCustomIntFromBool(Bool=$User.bisHeadHidden, True INT=1, False INT=0)) {isSetParameters=true, module=SetVariables_#5}
  ]
  rendererClasses [&NiagaraSpriteRendererProperties]
  renderers [{rendererClass=&NiagaraSpriteRendererProperties}]
 }
 FeatherSimulationProto {
  emitterUpdateScript [
   &EmitterState(
    Inactive Response=::NewEnumerator1
    Life Cycle Mode=::NewEnumerator1
    Loop Behavior=::NewEnumerator1
    Loop Duration Mode=::NewEnumerator1
    UseLoopDelay=false
   )
   &SpawnBurst_Instantaneous(Use Spawn Probability=false, UseLoopCountLimit=false)
  ]
  particleSpawnScript [&InitializeParticle(Color Mode=::NewEnumerator0, Lifetime=5, Sprite Size Mode=::NewEnumerator3, UsePositionOffset=false, Write Material Random=true)]
  particleUpdateScript [
   &ParticleState(Kill Particles When Lifetime Has Elapsed=false, Let Infinitely Lived Particles Die When Emitter Deactivates=true)
   &KillParticles(Kill Particles=$User.DisableSim)
   &PointAttractionForce(
    Attraction Radius=100000
    Attraction Strength=70
    Attractor Position:Vector3f=$System.FSOrigin
    Kill Within Radius=false
    Module.Use Falloff=false
    UseAttractorOffset=false
   )
   &WindForce(
    Air Resistance=0.1
    Wind Speed=&TransformVector(Vector=<1 0 0>, Destination Space=::World)
    Wind Speed Scale=&Waveform("Waveform [1]"=::NewEnumerator12, "[1] Frequency"=3, "[1] Amplitude Min/Max"=<0 1>, "[1] Scale"=2500)
   )
   &Drag(Use Linear Drag=true, Use Rotational Drag=false)
   &SolveForcesAndVelocity(Acceleration Limit=1000, Clamp Velocity=false, Limit Acceleration=true)
   &self:LimitRange(Center:Vector3f=$System.FSOrigin, Pos:NiagaraPosition=$Particles.Position, Radius=100) {module=LimitRange}
   /SetVariables(Particles.Direct:Vector3f=&ConvertPositionToVector(
     Input Position=&Multiply_VectorByFloat(
      Vector=&NormalizeVector(
       Vector To Normalize=&Subtract_Vector(A=Linked() {linkedVariable {name=Particles.Position, type=NiagaraPosition}}, B=$System.FSOrigin)
       Safe Normalize=true
       Safe Fallback Threshold=0
       Safe Fallback Vector=<0 0 0.001>
      )
      Float=&VectorLength(Vector=&Subtract_Vector(A=Linked() {linkedVariable {name=Particles.Position, type=NiagaraPosition}}, B=$System.FSOrigin))
     )
     Convert To Absolute Large World Vector=false
    )) {isSetParameters=true}
   &SampleSkeletalMesh(
    AdditionalYawPitchRoll=false
    Bone Sampling Mode:ENiagaraBoneSamplingMode=::NewEnumerator4
    Skeletal Mesh:NiagaraDataInterfaceSkeletalMesh=$User.PlayerBody
   )
   /Niagara/Modules/ExportParticleData/ExportParticleDataToBlueprint(
    Condition To Export Data=true
    Delay Between Data Exports:NiagaraFloat=&self:PlatformFloat(
     A:NiagaraFloat=0.0286
     B:NiagaraFloat=0.0164
     Platform:NiagaraDataInterfacePlatformSet=DataInterface() {propertyValues {Platforms {
        qualityLevelMask=-32
        deviceProfileStates [
         {profileName=Switch, qualityLevelMask=4294967295, setQualityLevelMask=4294967295}
         {profileName=Mobile, qualityLevelMask=4294967295, setQualityLevelMask=4294967295}
        ]
       }}}
    )
    Export Particle Data Interface:NiagaraDataInterfaceExport=$User.Ex
    "Float To Send (As Struct Size)":NiagaraFloat=&Multiply_Float(
     A=&Multiply_Float(
      A=&Subtract_Float(
       A=&Sine(
        Normalized Angle=&Divide_Float(A=$System.InterpolateOverTime.InterpolateOverTime_Float.Moving Average, B=2)
        Period=1
        Scale=1
        Bias=0
       )
       B=0.5
      )
      B=0.1
     )
     B=$System.InterpolateOverTime001.InterpolateOverTime_Float.Moving Average
    )
    UseDelayTimer=true
    "Vector To Send (As Struct Velocity)":Vector3f=$Particles.Direct
    "Vector to Send (As Struct Position)":Vector3f=&self:QUat(New Quat 4f:Quat4f=&MultiplyQuaternion(
      Quaternion A=/Niagara/DynamicInputs/Rotation/InvertQuaternion(Quat:Quat4f=&self:ScratchDynamicInput(Default:Vector3f=<1 0 0>, New Vector 3f:Vector3f=<0 1 0>, Z Vector:Vector3f=<0 0 1>))
      Quaternion B=&self:ScratchDynamicInput(
       Default:Vector3f=&TransformVector(
        Vector=&RotateVectorByQuaternion(Vector To Rotate=<0 0 -1>, Quaternion=$Particles.SampleSkeletalMesh.SampledRotation)
        Source Space=::World
        Destination Space=::Local
       )
       New Vector 3f:Vector3f=&TransformVector(
        Vector=&RotateVectorByQuaternion(Vector To Rotate=<0 1 0>, Quaternion=$Particles.SampleSkeletalMesh.SampledRotation)
        Source Space=::World
        Destination Space=::Local
       )
       Z Vector:Vector3f=&TransformVector(
        Vector=&RotateVectorByQuaternion(Vector To Rotate=<1 0 0>, Quaternion=$Particles.SampleSkeletalMesh.SampledRotation)
        Source Space=::World
        Destination Space=::Local
       )
      )
     ))
   ) {module=ExportParticleDataToBlueprint003}
  ]
  rendererClasses [&NiagaraSpriteRendererProperties]
  renderers [{rendererClass=&NiagaraSpriteRendererProperties}]
 }
 HairFX_DarkGold {
  emitterUpdateScript [
   &EmitterState(Life Cycle Mode=::NewEnumerator1, Loop Behavior=::NewEnumerator1, Loop Duration Mode=::NewEnumerator1, UseLoopDelay=false)
   &SpawnBurst_Instantaneous(Spawn Probability=&OneMinusFloat(Float=$System.Sys_Variant), Use Spawn Probability=true, UseLoopCountLimit=false)
   &SpawnBurst_Instantaneous(
    Spawn Group=1
    Spawn Probability=&OneMinusFloat(Float=$System.Sys_Variant)
    Use Spawn Probability=true
    UseLoopCountLimit=false
   ) {module=SpawnBurst_Instantaneous001}
   &SpawnBurst_Instantaneous(
    Spawn Group=2
    Spawn Probability=&OneMinusFloat(Float=$System.Sys_Variant)
    Use Spawn Probability=true
    UseLoopCountLimit=false
   ) {module=SpawnBurst_Instantaneous002}
  ]
  particleSpawnScript [
   &InitializeParticle(
    Color Mode=::NewEnumerator0
    Lifetime=5
    Mesh Scale=<1.04 1 1>
    Mesh Scale Mode=::NewEnumerator4
    Sprite Size=&Multiply_Vector2D_ByFloat(
     Float=&SelectFloatFromArray(Direct Array Index=$Emitter.SpawnGroup, Float Selection Array=DataInterface() {propertyValues {FloatData [1,1,0.8]}})
     Vector2D=<22 40>
    )
    Sprite Size Mode=::NewEnumerator4
    UsePositionOffset=false
    Write Material Random=true
   )
   &InitialMeshOrientation(Mesh Axis To Orient=<0 0 1>, Mesh Orientation Mode=::NewEnumerator2, Rotation=<0.1 0 0>, Use Rotation Vector=true)
   /SetVariables(Particles.VisibilityTag=$Emitter.SpawnGroup) {isSetParameters=true, module=SetVariables_#6}
   &CameraOffset(Camera Offset Amount=&SelectFloatFromArray(Direct Array Index=$Emitter.SpawnGroup, Float Selection Array=DataInterface() {propertyValues {FloatData [0,0,8]}}))
  ]
  particleUpdateScript [
   &ParticleState(Kill Particles When Lifetime Has Elapsed=false)
   &SkeletalMeshLocation(
    Bone Sampling Mode=::NewEnumerator4
    MaskSpawnGroup=false
    Sampled Position Offset=&SelectVectorFromArray(Direct Array Index=$Particles.VisibilityTag, Vector Selection Array=DataInterface() {propertyValues {FloatData [<5 2 0>, <15 -5 0>]}})
    Skeletal Mesh=$User.SK_Mesh
   )
   /SetVariables(Particles.SpriteAlignment=&RotateVectorByQuaternion(Vector To Rotate=<1 -0.8 0>, Quaternion=$Particles.SkeletalMeshLocation.SampledRotation)) {isSetParameters=true, module=SetVariables_#7}
   &ScaleSpriteSize(Scale Sprite Size Mode=::NewEnumerator0)
   &ScaleColor(Scale Alpha=&Lerp_Float(A=1, B=0, Alpha=$System.SpeedLerp), ScaleA=true, ScaleRGB=true)
   /SetVariables(Particles.VisibilityTag=&MakeCustomIntFromBool(Bool=$User.bisHeadHidden, True INT=3, False INT=0)) {isSetParameters=true, module=SetVariables_#5}
  ]
  rendererClasses [&NiagaraMeshRendererProperties, &NiagaraSpriteRendererProperties]
  renderers [
   {rendererClass=&NiagaraMeshRendererProperties, rendererIndex=0}
   {rendererClass=&NiagaraSpriteRendererProperties, rendererIndex=1}
   {rendererClass=&NiagaraSpriteRendererProperties, rendererIndex=2}
  ]
 }
 HairFX_DefaultPurple {
  emitterUpdateScript [
   &EmitterState(Life Cycle Mode=::NewEnumerator1, Loop Behavior=::NewEnumerator1, Loop Duration Mode=::NewEnumerator1, UseLoopDelay=false)
   &SpawnBurst_Instantaneous(Spawn Probability=$System.Sys_Variant, Use Spawn Probability=true, UseLoopCountLimit=false)
   &SpawnBurst_Instantaneous(Spawn Group=1, Spawn Probability=$System.Sys_Variant, Use Spawn Probability=true, UseLoopCountLimit=false) {module=SpawnBurst_Instantaneous001}
   &SpawnBurst_Instantaneous(Spawn Group=2, Spawn Probability=$System.Sys_Variant, Use Spawn Probability=true, UseLoopCountLimit=false) {module=SpawnBurst_Instantaneous002}
  ]
  particleSpawnScript [
   &InitializeParticle(
    Color Mode=::NewEnumerator0
    Lifetime=5
    Mesh Scale=<1.04 1 1>
    Mesh Scale Mode=::NewEnumerator4
    Sprite Size=&Multiply_Vector2D_ByFloat(
     Float=&SelectFloatFromArray(Direct Array Index=$Emitter.SpawnGroup, Float Selection Array=DataInterface() {propertyValues {FloatData [1,1,0.8]}})
     Vector2D=<22 40>
    )
    Sprite Size Mode=::NewEnumerator4
    UsePositionOffset=false
    Write Material Random=true
   )
   &InitialMeshOrientation(Mesh Axis To Orient=<0 0 1>, Mesh Orientation Mode=::NewEnumerator2, Rotation=<0.1 0 0>, Use Rotation Vector=true)
   /SetVariables(Particles.VisibilityTag=$Emitter.SpawnGroup) {isSetParameters=true, module=SetVariables_#6}
   &CameraOffset(Camera Offset Amount=&SelectFloatFromArray(Direct Array Index=$Emitter.SpawnGroup, Float Selection Array=DataInterface() {propertyValues {FloatData [0,0,8]}}))
  ]
  particleUpdateScript [
   &ParticleState(Kill Particles When Lifetime Has Elapsed=false, Let Infinitely Lived Particles Die When Emitter Deactivates=true)
   &SkeletalMeshLocation(
    Bone Sampling Mode=::NewEnumerator4
    MaskSpawnGroup=false
    Sampled Position Offset=&SelectVectorFromArray(Direct Array Index=$Particles.VisibilityTag, Vector Selection Array=DataInterface() {propertyValues {FloatData [<5 2 0>, <15 -5 0>]}})
    Skeletal Mesh=$User.SK_Mesh
   )
   /SetVariables(Particles.SpriteAlignment=&RotateVectorByQuaternion(Vector To Rotate=<1 -0.8 0>, Quaternion=$Particles.SkeletalMeshLocation.SampledRotation)) {isSetParameters=true, module=SetVariables_#7}
   &ScaleSpriteSize(Scale Sprite Size Mode=::NewEnumerator0)
   &ScaleColor(Scale Alpha=&Lerp_Float(A=1, B=0, Alpha=$System.SpeedLerp), ScaleA=true, ScaleRGB=true)
   /SetVariables(Particles.VisibilityTag=&MakeCustomIntFromBool(Bool=$User.bisHeadHidden, True INT=3, False INT=$Particles.Initial.VisibilityTag)) {isSetParameters=true, module=SetVariables_#5}
  ]
  rendererClasses [&NiagaraMeshRendererProperties, &NiagaraSpriteRendererProperties]
  renderers [
   {rendererClass=&NiagaraMeshRendererProperties, rendererIndex=0}
   {rendererClass=&NiagaraSpriteRendererProperties, rendererIndex=1}
   {rendererClass=&NiagaraSpriteRendererProperties, rendererIndex=2}
  ]
 }
}
summary {
 systemName=NS_TheHerald_HairFX
 emitters [
  {
   emitterName=HairFX_DefaultPurple
   rendererClasses [&NiagaraMeshRendererProperties, &NiagaraSpriteRendererProperties]
  }
  {emitterName=FeatherSimulationProto, rendererClasses [&NiagaraSpriteRendererProperties]}
  {
   emitterName=HairFX_DarkGold
   rendererClasses [&NiagaraMeshRendererProperties, &NiagaraSpriteRendererProperties]
  }
  {emitterName=Embers_DarkGold, rendererClasses [&NiagaraSpriteRendererProperties]}
  {emitterName=Embers_DarkGoldLocal, rendererClasses [&NiagaraSpriteRendererProperties]}
  {emitterName=Embers_DefaultPurple, rendererClasses [&NiagaraSpriteRendererProperties]}
 ]
 userVariables [
  User.bisHeadHidden {type=NiagaraBool}
  User.bVariant {type=NiagaraBool}
  User.DisableSim {type=NiagaraBool}
  User.Ex {
   type=NiagaraDataInterfaceExport
   description=""
   defaultValue {
    struct=NiagaraExt_VariableValue_DataInterface
    value {
     dataInterface=&self:NiagaraDataInterfaceExport_0
     dataInterfaceClass=/Script/Niagara.NiagaraDataInterfaceExport
    }
   }
  }
  User.IsMove {type=NiagaraBool}
  User.PlayerBody {
   type=NiagaraDataInterfaceSkeletalMesh
   defaultValue {
    struct=NiagaraExt_VariableValue_DataInterface
    value {dataInterface=&self:NiagaraDataInterfaceSkeletalMesh_1}
   }
  }
  User.SK_Mesh {
   type=NiagaraDataInterfaceSkeletalMesh
   defaultValue {
    struct=NiagaraExt_VariableValue_DataInterface
    value {dataInterface=&self:NiagaraDataInterfaceSkeletalMesh_0}
   }
  }
  User.TargetBP {
   type=Object
   description=""
   defaultValue {struct=NiagaraExt_VariableValue_Object, value {object=null, objectClass=&Object}}
  }
 ]
}
