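# Material function for The Herald character body, adding animated emissive to incoming material attributes and supplying UV distortion for a warping effect.
# Time-panned world-aligned noise, driven by pre-skinned position and normal to follow the animated mesh, modulates the emissive texture, optionally limited by an FX2 mask.
# A second panning noise, scaled by a selectable mask channel, outputs full and reduced UV offsets plus the distortion mask.
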MaterialFunction /BRCosmetics/Characters/Player/Female/Medium/Bodies/F_MED_TheHerald/Materials/MF_TheHerald
graph {nodes [
  n1 = TextureSampleParameter2D(UVs=NamedRerouteUsage()) {
   desc=red is eyes. Green is eye mask
   group=TheHerald
   parameterName=Emissive
   sortPriority=0
   texture=/BRCosmetics/Characters/Player/Female/Medium/Bodies/F_MED_TheHerald/Textures/T_F_MED_TheHerald_Body_E
  }
  n2 = ScalarParameter() {group=TheHerald, parameterName=TH_PanningEmissiveAdd}
  n3 = Time()
  n4 = Multiply(
   A=VectorParameter() {group=TheHerald, parameterName=TH_EmissivePosSpeed, defaultValue=<rgba 0 0 -5 1>}.RGB
   B=ScalarParameter() {defaultValue=0.5, group=TheHerald, parameterName=TH_EmissiveSpeed}
  ) {constB=0.1}
  n5 = NamedRerouteUsage()
  n6 = NamedRerouteUsage()
  n7 = ScalarParameter() {defaultValue=1, group=TheHerald, parameterName=TH_WarpProjectionContrast, sliderMax=1}
  n8 = MaterialFunctionCall(
   TextureObject=TextureObjectParameter() {group=TheHerald, parameterName=TH_EmissiveTexture, texture=/Game/Global/Textures/Default/Noise/T_Noise_Cloud}
   TextureSize=ScalarParameter() {defaultValue=15, group=TheHerald, parameterName=TH_EmissiveTextureSize}
   WorldPosition=Add(A=Multiply(A=@n3, B=@n4), B=@n5)
   World Space Normal=@n6
   ProjectionTransitionContrast=@n7
  ) {materialFunction=&WorldAlignedTexture}
  n9 = Power(Base=Multiply(A=ComponentMask(None=@n8.XY Texture) {r=true}) {constB=2})
  n10 = LinearInterpolate(
   A=@n1.RGB
   B=Multiply(
    A=Multiply(A=@n1.RGB, B=@n9)
    B=VectorParameter() {group=TheHerald, parameterName=TH_EmissiveNoiseTint, defaultValue=<rgba 1 0.5842 0.2218 0>}.RGB
   )
   Alpha=ScalarParameter() {group=TheHerald, parameterName=TH_PanningEmissiveLerp}
  )
  n11 = ScalarParameter() {defaultValue=1, group=TheHerald, parameterName=TH_PanningEmissiveMult}
  n12 = TextureSampleParameter2D(UVs=NamedRerouteUsage()) {
   desc=red is eyes. Green is eye mask
   group=TheHerald
   parameterName=FX2
   samplerType=SAMPLERTYPE_Masks
   sortPriority=0
   texture=/BRCosmetics/Characters/Player/Female/Medium/Bodies/F_MED_TheHerald/Textures/T_F_MED_TheHerald_Body_FX2
  }
  n13 = LinearInterpolate(
   A=Multiply(A=@n1.RGB, B=ScalarParameter() {defaultValue=1, group=TheHerald, parameterName=TH_StaticEmissiveMult}) {constB=0}
   B=Add(A=Multiply(A=@n1.RGB, B=@n2), B=Multiply(A=@n10, B=@n11) {constB=0})
   Alpha=StaticSwitchParameter(True=@n12.RGB, False=Constant() {r=1}) {defaultValue=true, group=TheHerald, parameterName=TH_useFX2}
  )
  n14 = SetMaterialAttributes!(
   MaterialAttributes=FunctionInput() {inputType=FunctionInput_MaterialAttributes}
   EmissiveColor=@n13
   Metallic
   Specular
   BaseColor
   Roughness
  ) {attributeSetTypes [
    #1
    #2
    #3
    #4
    #5
   ]}
  FunctionOutput(None=@n14)
  n15 = Multiply(
   A=VectorParameter() {group=TheHerald, parameterName=TH_WarpPosSpeed, defaultValue=<rgba 5 0 5 1>}.RGB
   B=ScalarParameter() {defaultValue=0.5, group=TheHerald, parameterName=TH_WarpSpeed}
  )
  n16 = MaterialFunctionCall(
   TextureObject=TextureObjectParameter() {group=TheHerald, parameterName=TH_WarpTexture, texture=&T_Noise_Cloud}
   TextureSize=ScalarParameter() {defaultValue=15, group=TheHerald, parameterName=TH_WarpTextureSize}
   WorldPosition=Add(A=Multiply(A=@n3, B=@n15), B=@n5)
   World Space Normal=@n6
   ProjectionTransitionContrast=@n7
  ) {materialFunction=&WorldAlignedTexture}
  n17 = Multiply(A=Subtract(A=ComponentMask(None=@n16.XY Texture) {g=true, r=true}) {constB=0.1}) {constB=0.01}
  n18 = ComponentMask(None=VectorParameter() {group=TheHerald, parameterName=TH_DistortionOffsetV}.RGB) {g=true, r=true}
  n19 = TextureSampleParameter2D() {
   desc=red is eyes. Green is eye mask
   group=TheHerald
   parameterName=SkinFX_Mask
   samplerType=SAMPLERTYPE_Masks
   sortPriority=0
   texture=/BRCosmetics/Characters/Player/Female/Medium/Bodies/F_MED_TheHerald/Textures/T_F_MED_TheHerald_Body_FX
  }
  n20 = TextureSampleParameter2D() {
   desc=red is eyes. Green is eye mask
   group=TheHerald
   parameterName=M
   samplerType=SAMPLERTYPE_Masks
   sortPriority=0
   texture=/BRCosmetics/Characters/Player/Female/Medium/Bodies/F_MED_TheHerald/Textures/T_F_MED_TheHerald_Body_M
  }
  n21 = ChannelMaskParameter(None=StaticSwitchParameter(True=@n19.RGB, False=@n20.RGB) {defaultValue=true, group=TheHerald, parameterName=TH_UseSkinFXMaskForDistortion}) {group=TheHerald, maskChannel=Green, parameterName=TH_DistortionChannel, defaultValue=<rgba 0 1 0 0>}
  n22 = Max(A=Saturate(None=@n21.R), B=ScalarParameter() {group=TheHerald, parameterName=TH_DistortionMinimum})
  n23 = Multiply(
   A=Multiply(A=Add(A=@n17, B=@n18), B=ScalarParameter() {defaultValue=0.3, group=TheHerald, parameterName=TH_DistortionStrength}) {constB=0.01}
   B=@n22
  )
  FunctionOutput(None=Multiply(A=@n23, B=ScalarParameter() {defaultValue=2, group=TheHerald, parameterName=TH_UVMult}) {constB=0.2}) {outputName=uvOffsets}
  n24 = Multiply(A=@n23, B=ScalarParameter() {defaultValue=0.666, group=TheHerald, parameterName=TH_ReducedUVMult}) {constB=0.2}
  FunctionOutput(None=@n24) {outputName=reducedUvOffsets}
  FunctionOutput(None=Saturate(None=@n22)) {outputName=distortionMask}
  Multiply(A=OneMinus(None=Fresnel() {exponent=1}), B=@n13)
  NamedRerouteDeclaration(None=Add(A=TextureCoordinate(), B=@n24)) {name=distortedUVs, nodeColor=<rgba 0.4531 0 1 1>}
  NamedRerouteDeclaration(None=VertexInterpolator(VS=PreSkinnedNormal()).PS) {name=PS_Normal, nodeColor=<rgba 1 0 0.8438 1>}
  NamedRerouteDeclaration(None=VertexInterpolator(VS=PreSkinnedPosition()).PS) {name=PS_Position, nodeColor=<rgba 1 0 0.1641 1>}
  n25 = TextureSampleParameter2D(UVs=NamedRerouteUsage()) {
   desc=red is eyes. Green is eye mask
   group=TheHerald
   parameterName=Diffuse
   sortPriority=0
   texture=/BRCosmetics/Characters/Player/Female/Medium/Bodies/F_MED_TheHerald/Textures/T_F_MED_TheHerald_Body_D
  }
  Reroute(None=NamedRerouteDeclaration(None=@n25.RGB) {name=diffuse, nodeColor=<rgba 0.4062 1 0 1>})
  ScalarParameter() {group=TheHerald, parameterName=TH_DistortionOffset}
 ]}
