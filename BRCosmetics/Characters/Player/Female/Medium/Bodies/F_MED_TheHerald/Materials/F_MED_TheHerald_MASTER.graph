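# Master material for The Herald's female medium character body, supporting Default Lit, Subsurface Profile and Cloth shading.
# MF_TheHerald sits between the shared character PreFX and PostFX functions, followed by optional cloth fuzz and thin film, then emissive fresnel.
# Its UV offsets distort PreFX and SkinFX mask sampling at two strengths, applied only in the pixel shader via ShaderStageSwitch.
# A static switch displays the distortion mask for debugging.
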Material /BRCosmetics/Characters/Player/Female/Medium/Bodies/F_MED_TheHerald/Materials/F_MED_TheHerald_MASTER
graph {
 properties {
  bUseMaterialAttributes=true
  bUsedWithMeshDeformer=true
  bUsedWithMorphTargets=true
  bUsedWithSkeletalMesh=true
  pixelDepthOffsetMode=PDOM_Legacy
  shadingModel=MSM_FromMaterialExpression
  subsurfaceProfile=&SS_HeroSkin_02
  usedShadingModels=Default Lit | Subsurface Profile | Cloth
 }
 nodes [
  n1 = Constant()
  n2 = MaterialFunctionCall!(In=NamedRerouteUsage()) {materialFunction=&/MF_TheHerald}
  MakeMaterialAttributes!(
   BaseColor=@n1
   Metallic=@n1
   Specular=@n1
   Roughness=@n1
   Anisotropy
   EmissiveColor=GetMaterialAttributes(None=@n2.Result) {attributeGetTypes [#1]}.Emissive Color
   Opacity
   OpacityMask
   Normal
   Tangent
   WorldPositionOffset
   SubsurfaceColor
   ClearCoat
   ClearCoatRoughness
   AmbientOcclusion
   Refraction
   CustomizedUVs_0
   CustomizedUVs_1
   CustomizedUVs_2
   CustomizedUVs_3
   CustomizedUVs_4
   CustomizedUVs_5
   CustomizedUVs_6
   CustomizedUVs_7
   PixelDepthOffset
   ShadingModel
   Displacement
  )
  n3 = TextureCoordinate()
  n4 = ShaderStageSwitch(PixelShader=Add(A=NamedRerouteUsage(), B=@n3), VertexShader=@n3)
  n5 = MaterialFunctionCall(TexCoords=@n4) {materialFunction=&BaseFN_CharacterFunction_PreFX}
  MaterialFunctionCall(AO=@n5.AO, Distance Blend=@n5.AO, Fresnel=@n5.AO, Gmap=NamedRerouteUsage(), Pre Skinned Local Position=@n5.AO, Material=NamedRerouteUsage()) {materialFunction=&BaseFN_CharacterFunction_PostFX}
  n6 = StaticComponentMaskParameter(None=NamedRerouteUsage()) {defaultR=true, group=Skin FX, parameterName=ThinFilm_Channel}
  n7 = NamedRerouteUsage()
  n8 = StaticSwitchParameter(
   True=MaterialFunctionCall!(
    Alpha=StaticComponentMaskParameter(None=@n7) {defaultR=true, group=Skin FX, parameterName=Cloth Channel}
    Gmap=NamedRerouteUsage()
    Material=@n2.Result
   ) {materialFunction=/BRCosmetics/Weapons/WeaponSkins/MaterialFunction/ClothFuzz_mm}.Result
   False=@n2.Result
  ) {group=Skin FX, parameterName=Use Cloth fuzz}
  n9 = MaterialFunctionCall!(In=StaticSwitchParameter(
    True=MaterialFunctionCall!(Alpha=@n6, Material Attributes=@n8) {materialFunction=/Game/Weapons/WeaponSkins/MaterialFunction/ThinFilmFunction}.Result
    False=@n8
   ) {defaultValue=true, group=Skin FX, parameterName=Use Thin Film}) {materialFunction=/Game/Weapons/WeaponSkins/MaterialFunction/MF_EmissiveFresnel}
  n10 = Constant()
  n11 = SetMaterialAttributes!(MaterialAttributes=@n9.Result, Base Color=@n2.distortionMask, Metallic=@n10, Specular=@n10, Roughness=@n10, Emissive Color=@n10) {attributeSetTypes [
    #2
    #3
    #4
    #5
    #1
   ]}
  NamedRerouteDeclaration(None=StaticSwitchParameter(True=@n11, False=@n9.Result) {group=TheHerald, parameterName=DebugHeraldDistortionMask}) {name=MaterialOut, nodeColor=<rgba 0 1 0.059 1>}
  NamedRerouteDeclaration(None=@n5.Result) {name=MaterialIn, nodeColor=<rgba 0.2656 1 0 1>}
  NamedRerouteDeclaration(None=@n5.Gmap) {name=Gmap, nodeColor=<rgba 1 0 0.7266 1>}
  n12 = TextureCoordinate()
  n13 = Add(A=NamedRerouteUsage(), B=@n12)
  n14 = TextureObjectParameter() {group=Skin FX, parameterName=SkinFX_Mask}
  NamedRerouteDeclaration(None=TextureSample(UVs=ShaderStageSwitch(PixelShader=@n13, VertexShader=@n12), Tex=@n14).RGB) {name=SkinFX_Mask_RGB_distorted, nodeColor=<rgba 1 0 0.3047 1>}
  NamedRerouteDeclaration(None=@n2.uvOffsets) {name=uvOffsets, nodeColor=<rgba 0 1 0.5078 1>}
  NamedRerouteDeclaration(None=@n2.reducedUvOffsets) {name=reducedUVOffsets, nodeColor=<rgba 0.5234 0 1 1>}
  NamedRerouteDeclaration(None=TextureSample(UVs=@n4, Tex=@n14).RGB) {name=SkinFX_Mask_RGB_LessDistorted, nodeColor=<rgba 0 0.2969 1 1>}
  NamedRerouteUsage()
  NamedRerouteUsage()
 ]
}
