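# Physics asset for the F_MED_TheHerald player body, setting up secondary motion for the dress panels and hip and right-arm petals.
# Capsules on pelvis, legs and right forearm form the body, linked by angularly free constraints.
# Capsule and box chains hang with limited swing from pelvis, right thigh or forearm twist bone, most with locked twist.
# An angularly free cross-link ties the middle side and side-back dress bodies together.
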PhysicsAsset /BRCosmetics/Characters/Player/Female/Medium/Bodies/F_MED_TheHerald/Meshes/F_MED_TheHerald_Physics
bodies {
 calf_l [{shapeType=Capsule, length=40, radius=9, shapeName=calf_r_capsule, center=<-21.9316 -3.1528 0>, rotation=<rot 90 180 175.0001>}]
 calf_r [{shapeType=Capsule, length=40, radius=8, shapeName=calf_r_capsule, center=<21.9316 3.1528 0>, rotation=<rot 90 180 175.0001>}]
 dyn_L_back_dress_1 [{shapeType=Box, extentX=17, extentY=3.5, extentZ=17, shapeName=dyn_front_dress_2_box, center=<5 0.6 0>, rotation=<rot 0 0 90>}]
 dyn_L_back_dress_2 [{shapeType=Box, extentX=17, extentY=3.5, extentZ=17, shapeName=dyn_front_dress_2_box, center=<5 0.6 0>, rotation=<rot 0 0 90>}]
 dyn_R_arm_petal_A [{shapeType=Capsule, length=2, radius=6, shapeName=dyn_R_arm_petal_A_capsule, center=<3 0 0>, rotation=<rot 90 0 0>}]
 dyn_R_arm_petal_B1 [{shapeType=Capsule, length=2, radius=6, shapeName=dyn_R_arm_petal_B1_capsule, center=<3 0 0>, rotation=<rot 90 0 0>}]
 dyn_R_arm_petal_B2 [{shapeType=Capsule, length=2, radius=6, shapeName=dyn_R_arm_petal_B2_capsule, center=<3 0 0>, rotation=<rot 90 0 0>}]
 dyn_R_backHip_petal_1 [{shapeType=Capsule, length=5, radius=4, shapeName=dyn_R_hip_petal_1_capsule, center=<3 0 0>, rotation=<rot 90 0 0>}]
 dyn_R_backHip_petal_2 [{shapeType=Capsule, length=5, radius=4, shapeName=dyn_R_hip_petal_1_capsule, center=<3 0 0>, rotation=<rot 90 0 0>}]
 dyn_R_hip_petal_1 [{shapeType=Capsule, length=5, radius=4.5, shapeName=dyn_R_hip_petal_1_capsule, center=<3 0 0>, rotation=<rot 90 0 0>}]
 dyn_R_hip_petal_2 [{shapeType=Capsule, length=5, radius=4, shapeName=dyn_R_hip_petal_2_capsule, center=<3 0 0>, rotation=<rot 90 0 0>}]
 dyn_back_dress_1 [{shapeType=Capsule, length=10, radius=3, shapeName=dyn_back_dress_1_capsule, center=<3 0 0>, rotation=<rot 90 0 0>}]
 dyn_back_dress_2 [{shapeType=Capsule, length=10, radius=4, shapeName=dyn_back_dress_2_capsule, center=<3 0 0>, rotation=<rot 90 0 0>}]
 dyn_back_dress_3 [{shapeType=Capsule, length=10, radius=5.4, shapeName=dyn_back_dress_3_capsule, center=<3 0 0>, rotation=<rot 90 0 0>}]
 dyn_front_dress_1 [{shapeType=Box, extentX=17, extentY=3.5, extentZ=17, shapeName=dyn_front_dress_2_box, center=<4 0.6 0>}]
 dyn_front_dress_2 [{shapeType=Box, extentX=17, extentY=3.5, extentZ=17, shapeName=dyn_front_dress_2_box, center=<4 0.6 0>}]
 dyn_sideBack_dress_1 [{shapeType=Capsule, length=10, radius=3, shapeName=dyn_sideBack_dress_1_capsule, center=<3 0 -2>, rotation=<rot 90 0 0>}]
 dyn_sideBack_dress_2 [{shapeType=Capsule, length=10, radius=3, shapeName=dyn_sideBack_dress_2_capsule, center=<3 0 0>, rotation=<rot 90 0 0>}]
 dyn_sideBack_dress_3 [{shapeType=Capsule, length=10, radius=4, shapeName=dyn_sideBack_dress_3_capsule, center=<3 0 0>, rotation=<rot 90 0 0>}]
 dyn_side_dress_1 [{shapeType=Capsule, length=10, radius=3, shapeName=dyn_side_dress_1_capsule, center=<3 0 0>, rotation=<rot 90 0 0>}]
 dyn_side_dress_2 [{shapeType=Capsule, length=10, radius=3, shapeName=dyn_side_dress_2_capsule, center=<3 0 0>, rotation=<rot 90 0 0>}]
 dyn_side_dress_3 [{shapeType=Capsule, length=20, radius=5, shapeName=dyn_side_dress_3_capsule, center=<3 0 0>, rotation=<rot 90 0 0>}]
 foot_r [{shapeType=Capsule, length=10, radius=8, shapeName=foot_r_capsule, center=<2.7125 -5.6389 -0.0797>, rotation=<rot 0 -1.2903 -88.8158>}]
 lowerarm_r [{shapeType=Capsule, length=1, radius=1, shapeName=lowerarm_r_capsule, rotation=<rot 90 0 0>}]
 lowerarm_twist_01_r [{shapeType=Capsule, length=1, radius=1, shapeName=lowerarm_twist_01_r_capsule}]
 pelvis [{shapeType=Capsule, length=3, radius=11, shapeName=pelvis_capsule, center=<-7.6934 0 0>}]
 thigh_l [
  {shapeType=Capsule, length=25, radius=7, shapeName=thigh_l_capsule, center=<-23.3673 0 0>, rotation=<rot 90 0 0>}
  {shapeType=Capsule, length=10, radius=8, shapeName=thigh_l_capsule, center=<-34.2311 -0.1024 0>, rotation=<rot 90 0 0>}
 ]
 thigh_r [{shapeType=Capsule, length=22, radius=8, shapeName=thigh_r_capsule, center=<23.3673 0 0>, rotation=<rot 90 0 0>}]
}
constraints [
 {bone1Name=dyn_R_hip_petal_1, bone2Name=pelvis, swing1LimitDegrees=20, swing2LimitDegrees=45, twistLimitDegrees=45, twistMotion=Locked}
 {
  bone1Name=dyn_R_hip_petal_2
  bone2Name=dyn_R_hip_petal_1
  swing1LimitDegrees=20
  swing2LimitDegrees=45
  twistLimitDegrees=45
  twistMotion=Locked
 }
 {bone1Name=dyn_back_dress_1, bone2Name=pelvis, swing1LimitDegrees=45, swing2LimitDegrees=45, twistLimitDegrees=45, twistMotion=Locked}
 {
  bone1Name=dyn_back_dress_2
  bone2Name=dyn_back_dress_1
  swing1LimitDegrees=45
  swing2LimitDegrees=45
  twistLimitDegrees=45
  twistMotion=Locked
 }
 {
  bone1Name=dyn_back_dress_3
  bone2Name=dyn_back_dress_2
  swing1LimitDegrees=45
  swing2LimitDegrees=45
  twistLimitDegrees=45
  twistMotion=Locked
 }
 {bone1Name=dyn_front_dress_2, bone2Name=dyn_front_dress_1, swing1LimitDegrees=45, swing2LimitDegrees=45, twistLimitDegrees=20}
 {bone1Name=dyn_sideBack_dress_1, bone2Name=pelvis, swing1LimitDegrees=45, swing2LimitDegrees=45, twistLimitDegrees=45, twistMotion=Locked}
 {
  bone1Name=dyn_sideBack_dress_2
  bone2Name=dyn_sideBack_dress_1
  swing1LimitDegrees=45
  swing2LimitDegrees=45
  twistLimitDegrees=45
  twistMotion=Locked
 }
 {
  bone1Name=dyn_sideBack_dress_3
  bone2Name=dyn_sideBack_dress_2
  swing1LimitDegrees=45
  swing2LimitDegrees=45
  twistLimitDegrees=45
  twistMotion=Locked
 }
 {bone1Name=dyn_R_arm_petal_A, bone2Name=lowerarm_twist_01_r, swing1LimitDegrees=20, swing2LimitDegrees=20, twistLimitDegrees=0}
 {bone1Name=dyn_R_arm_petal_B1, bone2Name=lowerarm_twist_01_r, swing1LimitDegrees=20, swing2LimitDegrees=20, twistLimitDegrees=0}
 {bone1Name=dyn_R_arm_petal_B2, bone2Name=dyn_R_arm_petal_B1, swing1LimitDegrees=20, swing2LimitDegrees=20, twistLimitDegrees=0}
 {
  bone1Name=lowerarm_r
  bone2Name=pelvis
  swing1LimitDegrees=45
  swing1Motion=Free
  swing2LimitDegrees=45
  swing2Motion=Free
  twistLimitDegrees=45
  twistMotion=Free
 }
 {bone1Name=dyn_side_dress_1, bone2Name=thigh_r, swing1LimitDegrees=45, swing2LimitDegrees=45, twistLimitDegrees=45, twistMotion=Locked}
 {
  bone1Name=dyn_side_dress_2
  bone2Name=dyn_side_dress_1
  swing1LimitDegrees=45
  swing2LimitDegrees=45
  twistLimitDegrees=45
  twistMotion=Locked
 }
 {
  bone1Name=dyn_side_dress_3
  bone2Name=dyn_side_dress_2
  swing1LimitDegrees=45
  swing2LimitDegrees=45
  twistLimitDegrees=45
  twistMotion=Locked
 }
 {
  bone1Name=thigh_l
  bone2Name=pelvis
  swing1LimitDegrees=45
  swing1Motion=Free
  swing2LimitDegrees=45
  swing2Motion=Free
  twistLimitDegrees=45
  twistMotion=Free
 }
 {
  bone1Name=thigh_r
  bone2Name=pelvis
  swing1LimitDegrees=45
  swing1Motion=Free
  swing2LimitDegrees=45
  swing2Motion=Free
  twistLimitDegrees=45
  twistMotion=Free
 }
 {
  bone1Name=calf_r
  bone2Name=thigh_r
  swing1LimitDegrees=45
  swing1Motion=Free
  swing2LimitDegrees=45
  swing2Motion=Free
  twistLimitDegrees=45
  twistMotion=Free
 }
 {bone1Name=dyn_front_dress_1, bone2Name=thigh_r, swing1LimitDegrees=45, swing2LimitDegrees=45, twistLimitDegrees=20}
 {
  bone1Name=calf_l
  bone2Name=thigh_l
  swing1LimitDegrees=45
  swing1Motion=Free
  swing2LimitDegrees=45
  swing2Motion=Free
  twistLimitDegrees=45
  twistMotion=Free
 }
 {bone1Name=dyn_L_back_dress_1, bone2Name=pelvis, swing1LimitDegrees=45, swing2LimitDegrees=45, twistLimitDegrees=45, twistMotion=Locked}
 {
  bone1Name=dyn_L_back_dress_2
  bone2Name=dyn_L_back_dress_1
  swing1LimitDegrees=45
  swing2LimitDegrees=45
  twistLimitDegrees=45
  twistMotion=Locked
 }
 {
  bone1Name=foot_r
  bone2Name=calf_r
  swing1LimitDegrees=45
  swing1Motion=Free
  swing2LimitDegrees=45
  swing2Motion=Free
  twistLimitDegrees=45
  twistMotion=Free
 }
 {bone1Name=dyn_R_backHip_petal_1, bone2Name=pelvis, swing1LimitDegrees=20, swing2LimitDegrees=45, twistLimitDegrees=45, twistMotion=Locked}
 {
  bone1Name=dyn_R_backHip_petal_2
  bone2Name=dyn_R_backHip_petal_1
  swing1LimitDegrees=20
  swing2LimitDegrees=45
  twistLimitDegrees=45
  twistMotion=Locked
 }
 {
  bone1Name=lowerarm_twist_01_r
  bone2Name=lowerarm_r
  swing1LimitDegrees=45
  swing1Motion=Free
  swing2LimitDegrees=45
  swing2Motion=Free
  twistLimitDegrees=45
  twistMotion=Free
 }
 {
  bone1Name=dyn_side_dress_2
  bone2Name=dyn_sideBack_dress_2
  swing1LimitDegrees=45
  swing1Motion=Free
  swing2LimitDegrees=45
  swing2Motion=Free
  twistLimitDegrees=45
  twistMotion=Free
 }
]
