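# Animation Blueprint for The Herald female body cosmetic, adding secondary motion to leaves, petals and dress.
# It copies the parent mesh pose, constrains backpack and cape bones, then runs AnimDynamics and a RigidBody simulation.
# Corrective bone transforms and constraints then reshape the dress for thigh angle, crouch, skydive, front end and knee slide.
# Event Graph updates state-driven gravity and blend alphas, switching RigidBody off during knee slides.
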AnimBlueprint /BRCosmetics/Characters/Player/Female/Medium/Bodies/F_MED_TheHerald/Meshes/F_MED_TheHerald_AnimBP
target_skeleton=&/F_MED_TheHerald_Skeleton
graphs [
 AnimGraph {nodes [
   RigidBody!(ComponentPose, Alpha, OverrideWorldGravity:Vector="0,0,-980", ExternalForce) -> (Pose) {
    bClampLinearTranslationLimitToRefPose=true
    bOverrideWorldGravity=true
    overridePhysicsAsset=&/F_MED_TheHerald_Physics
    simulationSpace=BaseBoneSpace
    alphaBoolBlend {blendOption=Linear}
    alphaScaleBiasClamp {bInterpResult=true}
    baseBoneRef {boneName=root}
    componentLinearAccScale=<0.3 0.3 0.3>
    componentLinearVelScale=<0.4 0.4 0.4>
    overrideWorldGravity {z=-980}
    simSpaceSettings {maxAngularAcceleration=9, maxAngularVelocity=9, maxLinearAcceleration=9, maxLinearVelocity=9}
   }
   n1 = "Constraint - attach_backpack (PPPPPP)"!(
    ComponentPose=Local To Component(LocalPose=Copy Pose From Mesh() {bUseAttachedParent=true})
    Alpha=1
    ConstraintWeights_0=0.051
    ConstraintWeights_1:float=0.034
    ConstraintWeights_2:float=0.384
    ConstraintWeights_3:float=0.451
    ConstraintWeights_4:float=0.011
    ConstraintWeights_5:float=0.07
   ) -> (Pose) {
    alphaBoolBlend {blendOption=Linear}
    boneToModify {boneName=attach_backpack}
    constraintSetup [
     {targetBone {boneName=clavicle_l}}
     {targetBone {boneName=clavicle_r}}
     {targetBone {boneName=spine_04}}
     {targetBone {boneName=spine_05}}
     {targetBone {boneName=spine_02}}
     {targetBone {boneName=spine_03}}
    ]
    constraintWeights [1,1,1,1,1,1]
   }
   n2 = "Constraint - attach_cape (PPPPPPP)"!(
    ComponentPose=@n1
    Alpha=1
    ConstraintWeights_0=0.748
    ConstraintWeights_1:float=0.006
    ConstraintWeights_2:float=0.121
    ConstraintWeights_3:float=0.015
    ConstraintWeights_4:float=0.065
    ConstraintWeights_5:float=0.084
    ConstraintWeights_6:float=0.01
   ) -> (Pose) {
    alphaBoolBlend {blendOption=Linear}
    boneToModify {boneName=attach_cape}
    constraintSetup [
     {targetBone {boneName=spine_05}}
     {targetBone {boneName=neck_02}}
     {targetBone {boneName=clavicle_r}}
     {targetBone {boneName=spine_04}}
     {targetBone {boneName=neck_01}}
     {targetBone {boneName=clavicle_l}}
     {targetBone {boneName=spine_03}}
    ]
    constraintWeights [1,1,1,1,1,1,1]
   }
   n3 = Anim Dynamics - Bone: dyn_neckLeaf_A(ComponentPose=@n2, Alpha=1, ExternalForce="0,0,0") {
    angularDampingOverride=1
    angularSpringConstant=3000
    bAngularSpring=true
    bOverrideAngularDamping=true
    bUseGravityOverride=true
    gravityScale=0.35
    simulationSpace=BoneRelative
    alphaBoolBlend {blendOption=Linear}
    boundBone {boneName=dyn_neckLeaf_A}
    componentLinearAccScale=<1 1 1>
    componentLinearVelScale=<1 1 1>
    gravityOverride {z=-980}
    physicsBodyDefinitions [{
      boundBone {boneName=dyn_neckLeaf_A}
      boxExtents=<10 10 10>
      constraintSetup {angularLimitsMax=<0 10 10>, angularLimitsMin=<0 -10 -10>, angularTarget=<1 0 0>}
      localJointOffset=<1 0 0>
     }]
    relativeSpaceBone {boneName=spine_05}
   }
   n4 = Anim Dynamics - Bone: dyn_neckLeaf_B(ComponentPose=@n3, Alpha=1, ExternalForce="0,0,0") {
    angularDampingOverride=1
    angularSpringConstant=3000
    bAngularSpring=true
    bOverrideAngularDamping=true
    bUseGravityOverride=true
    gravityScale=0.35
    simulationSpace=BoneRelative
    alphaBoolBlend {blendOption=Linear}
    boundBone {boneName=dyn_neckLeaf_B}
    componentLinearAccScale=<1 1 1>
    componentLinearVelScale=<1 1 1>
    gravityOverride {z=-980}
    physicsBodyDefinitions [{
      boundBone {boneName=dyn_neckLeaf_B}
      boxExtents=<10 10 10>
      constraintSetup {angularLimitsMax=<0 10 10>, angularLimitsMin=<0 -10 -10>, angularTarget=<1 0 0>}
      localJointOffset=<1 0 0>
     }]
    relativeSpaceBone {boneName=spine_05}
   }
   n5 = Anim Dynamics - Bone: dyn_neckLeaf_C(ComponentPose=@n4, Alpha=1, ExternalForce="0,0,0") {
    angularDampingOverride=1
    angularSpringConstant=3000
    bAngularSpring=true
    bOverrideAngularDamping=true
    bUseGravityOverride=true
    gravityScale=0.35
    simulationSpace=BoneRelative
    alphaBoolBlend {blendOption=Linear}
    boundBone {boneName=dyn_neckLeaf_C}
    componentLinearAccScale=<1 1 1>
    componentLinearVelScale=<1 1 1>
    gravityOverride {z=-980}
    physicsBodyDefinitions [{
      boundBone {boneName=dyn_neckLeaf_C}
      boxExtents=<10 10 10>
      constraintSetup {angularLimitsMax=<0 10 10>, angularLimitsMin=<0 -10 -10>, angularTarget=<1 0 0>}
      localJointOffset=<1 0 0>
     }]
    relativeSpaceBone {boneName=spine_05}
   }
   n6 = Anim Dynamics - Bone: dyn_neckLeaf_D_base(ComponentPose=@n5, Alpha=1, ExternalForce="0,0,0") {
    angularDampingOverride=1
    angularSpringConstant=3000
    bAngularSpring=true
    bOverrideAngularDamping=true
    bUseGravityOverride=true
    gravityScale=0.35
    simulationSpace=BoneRelative
    alphaBoolBlend {blendOption=Linear}
    boundBone {boneName=dyn_neckLeaf_D_base}
    componentLinearAccScale=<1 1 1>
    componentLinearVelScale=<1 1 1>
    gravityOverride {z=-980}
    physicsBodyDefinitions [{
      boundBone {boneName=dyn_neckLeaf_D_base}
      boxExtents=<10 10 10>
      constraintSetup {angularLimitsMax=<0 10 10>, angularLimitsMin=<0 0 -10>, angularTarget=<1 0 0>}
      localJointOffset=<1 0 0>
     }]
    relativeSpaceBone {boneName=spine_05}
   }
   n7 = Anim Dynamics - Bone: dyn_neckLeaf_D_1(ComponentPose=@n6, Alpha=1, ExternalForce="0,0,0") {
    angularDampingOverride=1
    angularSpringConstant=3000
    bAngularSpring=true
    bOverrideAngularDamping=true
    bUseGravityOverride=true
    gravityScale=0.35
    simulationSpace=BoneRelative
    alphaBoolBlend {blendOption=Linear}
    boundBone {boneName=dyn_neckLeaf_D_1}
    componentLinearAccScale=<1 1 1>
    componentLinearVelScale=<1 1 1>
    gravityOverride {z=-980}
    physicsBodyDefinitions [{
      boundBone {boneName=dyn_neckLeaf_D_1}
      boxExtents=<10 10 10>
      constraintSetup {angularLimitsMax=<0 10 10>, angularLimitsMin=<0 -10 -10>, angularTarget=<1 0 0>}
      localJointOffset=<1 0 0>
     }]
    relativeSpaceBone {boneName=spine_05}
   }
   n8 = Anim Dynamics - Bone: dyn_neckLeaf_D_2(ComponentPose=@n7, Alpha=1, ExternalForce="0,0,0") {
    angularDampingOverride=1
    angularSpringConstant=3000
    bAngularSpring=true
    bOverrideAngularDamping=true
    bUseGravityOverride=true
    gravityScale=0.35
    simulationSpace=BoneRelative
    alphaBoolBlend {blendOption=Linear}
    boundBone {boneName=dyn_neckLeaf_D_2}
    componentLinearAccScale=<1 1 1>
    componentLinearVelScale=<1 1 1>
    gravityOverride {z=-980}
    physicsBodyDefinitions [{
      boundBone {boneName=dyn_neckLeaf_D_2}
      boxExtents=<10 10 10>
      constraintSetup {angularLimitsMax=<0 10 10>, angularLimitsMin=<0 -10 -10>, angularTarget=<1 0 0>}
      localJointOffset=<1 0 0>
     }]
    relativeSpaceBone {boneName=spine_05}
   }
   n9 = Anim Dynamics - Bone: dyn_neckLeaf_E(ComponentPose=@n8, Alpha=1, ExternalForce="0,0,0") {
    angularDampingOverride=1
    angularSpringConstant=3000
    bAngularSpring=true
    bOverrideAngularDamping=true
    bUseGravityOverride=true
    gravityScale=0.35
    simulationSpace=BoneRelative
    alphaBoolBlend {blendOption=Linear}
    boundBone {boneName=dyn_neckLeaf_E}
    componentLinearAccScale=<1 1 1>
    componentLinearVelScale=<1 1 1>
    gravityOverride {z=-980}
    physicsBodyDefinitions [{
      boundBone {boneName=dyn_neckLeaf_E}
      boxExtents=<10 10 10>
      constraintSetup {angularLimitsMax=<0 10 10>, angularLimitsMin=<0 -10 -10>, angularTarget=<1 0 0>}
      localJointOffset=<1 0 0>
     }]
    relativeSpaceBone {boneName=spine_05}
   }
   n10 = Anim Dynamics - Bone: dyn_neckLeaf_F(ComponentPose=@n9, Alpha=1, ExternalForce="0,0,0") {
    angularDampingOverride=1
    angularSpringConstant=3000
    bAngularSpring=true
    bOverrideAngularDamping=true
    bUseGravityOverride=true
    gravityScale=0.35
    simulationSpace=BoneRelative
    alphaBoolBlend {blendOption=Linear}
    boundBone {boneName=dyn_neckLeaf_F}
    componentLinearAccScale=<1 1 1>
    componentLinearVelScale=<1 1 1>
    gravityOverride {z=-980}
    physicsBodyDefinitions [{
      boundBone {boneName=dyn_neckLeaf_F}
      boxExtents=<10 10 10>
      constraintSetup {angularLimitsMax=<0 10 10>, angularLimitsMin=<0 -10 -10>, angularTarget=<1 0 0>}
      localJointOffset=<1 0 0>
     }]
    relativeSpaceBone {boneName=spine_05}
   }
   n11 = Anim Dynamics - Bone: dyn_neckLeaf_G(ComponentPose=@n10, Alpha=1, ExternalForce="0,0,0") {
    angularDampingOverride=1
    angularSpringConstant=3000
    bAngularSpring=true
    bOverrideAngularDamping=true
    bUseGravityOverride=true
    gravityScale=0.35
    simulationSpace=BoneRelative
    alphaBoolBlend {blendOption=Linear}
    boundBone {boneName=dyn_neckLeaf_G}
    componentLinearAccScale=<1 1 1>
    componentLinearVelScale=<1 1 1>
    gravityOverride {z=-980}
    physicsBodyDefinitions [{
      boundBone {boneName=dyn_neckLeaf_G}
      boxExtents=<10 10 10>
      constraintSetup {angularLimitsMax=<0 10 10>, angularLimitsMin=<0 -10 -10>, angularTarget=<1 0 0>}
      localJointOffset=<1 0 0>
     }]
    relativeSpaceBone {boneName=spine_05}
   }
   n12 = Anim Dynamics - Bone: dyn_secondaryPetal_1(ComponentPose=@n11, Alpha=1, ExternalForce="0,0,0") {
    angularDampingOverride=0.999
    angularSpringConstant=2000
    bAngularSpring=true
    bOverrideAngularDamping=true
    bUseGravityOverride=true
    gravityScale=0.35
    simulationSpace=RootRelative
    alphaBoolBlend {blendOption=Linear}
    boundBone {boneName=dyn_secondaryPetal_1}
    componentLinearAccScale=<1 1 1>
    componentLinearVelScale=<1 1 1>
    gravityOverride {z=-980}
    physicsBodyDefinitions [{
      boundBone {boneName=dyn_secondaryPetal_1}
      boxExtents=<15 15 15>
      constraintSetup {angularLimitsMax=<20 40 20>, angularLimitsMin=<-20 -60 -20>, angularTarget=<1 0 0>}
      localJointOffset=<1 0 0>
     }]
   }
   n13 = Anim Dynamics - Bone: dyn_secondaryPetal_2(ComponentPose=@n12, Alpha=1, ExternalForce="0,0,0") {
    angularDampingOverride=0.999
    angularSpringConstant=2000
    bAngularSpring=true
    bOverrideAngularDamping=true
    bUseGravityOverride=true
    gravityScale=0.35
    simulationSpace=RootRelative
    alphaBoolBlend {blendOption=Linear}
    boundBone {boneName=dyn_secondaryPetal_2}
    componentLinearAccScale=<1 1 1>
    componentLinearVelScale=<1 1 1>
    gravityOverride {z=-980}
    physicsBodyDefinitions [{
      boundBone {boneName=dyn_secondaryPetal_2}
      boxExtents=<15 15 15>
      constraintSetup {angularLimitsMax=<20 80 20>, angularLimitsMin=<-20 -80 -20>, angularTarget=<1 0 0>}
      localJointOffset=<1 0 0>
     }]
   }
   n14 = Anim Dynamics - Bone: dyn_secondaryPetal_3(ComponentPose=@n13, Alpha=1, ExternalForce="0,0,0") {
    angularDampingOverride=0.999
    angularSpringConstant=2000
    bAngularSpring=true
    bOverrideAngularDamping=true
    bUseGravityOverride=true
    gravityScale=0.35
    simulationSpace=RootRelative
    alphaBoolBlend {blendOption=Linear}
    boundBone {boneName=dyn_secondaryPetal_3}
    componentLinearAccScale=<1 1 1>
    componentLinearVelScale=<1 1 1>
    gravityOverride {z=-980}
    physicsBodyDefinitions [{
      boundBone {boneName=dyn_secondaryPetal_3}
      boxExtents=<15 15 15>
      constraintSetup {angularLimitsMax=<20 80 20>, angularLimitsMin=<-20 -80 -20>, angularTarget=<1 0 0>}
      localJointOffset=<1 0 0>
     }]
   }
   n15 = RigidBody(ComponentPose=@n14, Alpha=Get Isnt Sliding() -> (Isnt Sliding:double), OverrideWorldGravity=Get GravityOverride()) {
    bClampLinearTranslationLimitToRefPose=true
    bOverrideWorldGravity=true
    overridePhysicsAsset=&F_MED_TheHerald_Physics
    simulationSpace=BaseBoneSpace
    alphaBoolBlend {blendOption=Linear}
    alphaScaleBiasClamp {bInterpResult=true}
    baseBoneRef {boneName=spine_02}
    componentLinearAccScale=<0.3 0.3 0.3>
    componentLinearVelScale=<0.4 0.4 0.4>
    overrideWorldGravity {z=-980}
    simSpaceSettings {maxAngularAcceleration=9, maxAngularVelocity=9, maxLinearAcceleration=9, maxLinearVelocity=9, velocityScaleZ=0.1, worldAlpha=1}
   }
   n16 = "Transform (Modify) Bone - Bone: dyn_L_back_dress_1"(
    ComponentPose=@n15
    Alpha=Get L thighRotatedBack() -> (L thighRotatedBack:double)
    Translation="0,1,-1"
    Rotation="-0.000,0.000,15"
    Scale="1,1,1"
   ) {
    rotationMode=BMM_Additive
    translationMode=BMM_Additive
    translationSpace=BCS_BoneSpace
    alphaBoolBlend {blendInTime=0.1, blendOption=Linear, blendOutTime=0.1}
    boneToModify {boneName=dyn_L_back_dress_1}
    rotation {roll=15}
    translation {y=1, z=-1}
   }
   n17 = Get bIsCrouching()
   n18 = "Constraint - dyn_sideBack_dress_1 (PP)"!(ComponentPose=@n16, bAlphaBoolEnabled:bool=@n17, ConstraintWeights_0=0.5, ConstraintWeights_1:float=0.5) -> (Pose) {
    alphaInputType=Bool
    alphaBoolBlend {blendOption=Linear}
    boneToModify {boneName=dyn_sideBack_dress_1}
    constraintSetup [{targetBone {boneName=dyn_side_dress_1}}, {targetBone {boneName=dyn_back_dress_1}}]
    constraintWeights [1,1]
   }
   n19 = "Constraint - dyn_sideBack_dress_2 (PP)"!(ComponentPose=@n18, bAlphaBoolEnabled:bool=@n17, ConstraintWeights_0=0.5, ConstraintWeights_1:float=0.5) -> (Pose) {
    alphaInputType=Bool
    alphaBoolBlend {blendOption=Linear}
    boneToModify {boneName=dyn_sideBack_dress_2}
    constraintSetup [{targetBone {boneName=dyn_side_dress_2}}, {targetBone {boneName=dyn_back_dress_2}}]
    constraintWeights [1,1]
   }
   n20 = "Constraint - dyn_sideBack_dress_3 (PP)"!(ComponentPose=@n19, bAlphaBoolEnabled:bool=@n17, ConstraintWeights_0=0.5, ConstraintWeights_1:float=0.5) -> (Pose) {
    alphaInputType=Bool
    alphaBoolBlend {blendOption=Linear}
    boneToModify {boneName=dyn_sideBack_dress_3}
    constraintSetup [{targetBone {boneName=dyn_side_dress_3}}, {targetBone {boneName=dyn_back_dress_3}}]
    constraintWeights [1,1]
   }
   n21 = Get bIsCrouching()
   n22 = "Transform (Modify) Bone - Bone: dyn_sideBack_dress_2"!(ComponentPose=@n20, bAlphaBoolEnabled:bool=@n21, Translation="0,0,0", Rotation="0,-20,50", Scale="1,1,1") -> (Pose) {
    alphaInputType=Bool
    rotationMode=BMM_Additive
    translationMode=BMM_Additive
    alphaBoolBlend {blendInTime=0.1, blendOption=Linear, blendOutTime=0.1}
    boneToModify {boneName=dyn_sideBack_dress_2}
    rotation {roll=50, yaw=-20}
   }
   n23 = "Transform (Modify) Bone - Bone: dyn_side_dress_2"!(ComponentPose=@n22, bAlphaBoolEnabled:bool=@n21, Translation="0,0,0", Rotation="-10.000,0.000,-0.000", Scale="1,1,1") -> (Pose) {
    alphaInputType=Bool
    rotationMode=BMM_Additive
    translationMode=BMM_Additive
    alphaBoolBlend {blendInTime=0.1, blendOption=Linear, blendOutTime=0.1}
    boneToModify {boneName=dyn_side_dress_2}
    rotation {pitch=-10}
   }
   n24 = "Constraint - dyn_R_hip_petal_2 (P)"!(ComponentPose=@n23, bAlphaBoolEnabled:bool=@n21, ConstraintWeights_0=1) -> (Pose) {
    alphaInputType=Bool
    alphaBoolBlend {blendOption=Linear}
    boneToModify {boneName=dyn_R_hip_petal_2}
    constraintSetup [{targetBone {boneName=thigh_r}}]
    constraintWeights [1]
   }
   n25 = "Constraint - dyn_front_dress_2 (P)"!(ComponentPose=@n24, bAlphaBoolEnabled:bool=@n21, ConstraintWeights_0=1) -> (Pose) {
    alphaInputType=Bool
    alphaBoolBlend {blendOption=Linear}
    boneToModify {boneName=dyn_front_dress_2}
    constraintSetup [{targetBone {boneName=thigh_r}}]
    constraintWeights [1]
   }
   n26 = Get bIsSkyDiving()
   n27 = "Transform (Modify) Bone - Bone: dyn_back_dress_1"!(ComponentPose=@n25, bAlphaBoolEnabled:bool=@n26, Scale="1,1,1") -> (Pose) {
    alphaInputType=Bool
    rotationMode=BMM_Additive
    translationMode=BMM_Additive
    alphaBoolBlend {blendOption=Linear}
    boneToModify {boneName=dyn_back_dress_1}
    translation=<0 -3 3>
   }
   n28 = "Transform (Modify) Bone - Bone: dyn_sideBack_dress_1"!(ComponentPose=@n27, bAlphaBoolEnabled:bool=@n26, Scale="1,1,1") -> (Pose) {
    alphaInputType=Bool
    rotationMode=BMM_Additive
    translationMode=BMM_Additive
    alphaBoolBlend {blendOption=Linear}
    boneToModify {boneName=dyn_sideBack_dress_1}
    translation {y=-3, z=3}
   }
   n29 = "Transform (Modify) Bone - Bone: dyn_L_back_dress_1"!(ComponentPose=@n28, bAlphaBoolEnabled:bool=@n26, Scale="1,1,1") -> (Pose) {
    alphaInputType=Bool
    rotationMode=BMM_Additive
    translationMode=BMM_Additive
    alphaBoolBlend {blendOption=Linear}
    boneToModify {boneName=dyn_L_back_dress_1}
    translation=<0 -3 10>
   }
   n30 = Get bIsInFrontEnd()
   n31 = "Transform (Modify) Bone - Bone: dyn_front_dress_1"!(ComponentPose=@n29, bAlphaBoolEnabled:bool=@n30, Scale="1,1,1") -> (Pose) {
    alphaInputType=Bool
    rotationMode=BMM_Additive
    translationMode=BMM_Additive
    alphaBoolBlend {blendOption=Linear}
    boneToModify {boneName=dyn_front_dress_1}
    rotation {roll=30}
   }
   n32 = "Transform (Modify) Bone - Bone: dyn_sideBack_dress_1"(ComponentPose=@n31, bAlphaBoolEnabled=@n30, Scale="1,1,1") {
    alphaInputType=Bool
    rotationMode=BMM_Additive
    translationMode=BMM_Additive
    alphaBoolBlend {blendOption=Linear}
    boneToModify {boneName=dyn_sideBack_dress_1}
    rotation {roll=-20}
   }
   n33 = "Transform (Modify) Bone - Bone: dyn_sideBack_dress_2"!(ComponentPose=@n32, bAlphaBoolEnabled=True, Scale="1,1,1") -> (Pose) {
    alphaInputType=Bool
    rotationMode=BMM_Additive
    translationMode=BMM_Additive
    alphaBoolBlend {blendOption=Linear}
    boneToModify {boneName=dyn_sideBack_dress_2}
    rotation {roll=25}
   }
   n34 = Get Is Sliding() -> (Is Sliding:float)
   n35 = "Transform (Modify) Bone - Bone: dyn_sideBack_dress_1"!(ComponentPose=@n33, Alpha:float=@n34, Translation:Vector="-6.973,4.188,-2.184", Rotation:Rotator="-74,-67,35", Scale="1,1,1") -> (Pose) {
    rotationMode=BMM_Additive
    rotationSpace=BCS_BoneSpace
    translationMode=BMM_Additive
    translationSpace=BCS_BoneSpace
    alphaBoolBlend {blendOption=Linear}
    alphaScaleBiasClamp {bInterpResult=true}
    boneToModify {boneName=dyn_sideBack_dress_1}
    rotation=<rot -74 -67 35>
    translation=<-6.973 4.188 -2.184>
   }
   n36 = "Transform (Modify) Bone - Bone: dyn_back_dress_1"!(ComponentPose=@n35, Alpha:float=@n34, Translation:Vector="0,0,0", Rotation:Rotator="-60.000,180.000,180.000", Scale="1,1,1") -> (Pose) {
    rotationMode=BMM_Additive
    rotationSpace=BCS_BoneSpace
    translationMode=BMM_Additive
    translationSpace=BCS_BoneSpace
    alphaBoolBlend {blendOption=Linear}
    alphaScaleBiasClamp {bInterpResult=true}
    boneToModify {boneName=dyn_back_dress_1}
    rotation=<rot -60 180 180>
   }
   n37 = "Transform (Modify) Bone - Bone: dyn_L_back_dress_1"!(ComponentPose=@n36, Alpha:float=@n34, Translation:Vector="1.726,-5.150,3.591", Rotation:Rotator="-45.5,178,-173", Scale="1,1,1") -> (Pose) {
    rotationMode=BMM_Additive
    rotationSpace=BCS_BoneSpace
    translationMode=BMM_Additive
    translationSpace=BCS_BoneSpace
    alphaBoolBlend {blendOption=Linear}
    alphaScaleBiasClamp {bInterpResult=true}
    boneToModify {boneName=dyn_L_back_dress_1}
    rotation=<rot -45.5 178 -173>
    translation=<1.726 -5.15 3.591>
   }
   n38 = "Transform (Modify) Bone - Bone: dyn_side_dress_1"(ComponentPose=@n37, Alpha=@n34, Translation="0,0,0", Rotation="-37,-69,2.729", Scale="1,1,1") {
    rotationMode=BMM_Additive
    rotationSpace=BCS_BoneSpace
    translationMode=BMM_Additive
    translationSpace=BCS_BoneSpace
    alphaBoolBlend {blendOption=Linear}
    alphaScaleBiasClamp {bInterpResult=true}
    boneToModify {boneName=dyn_side_dress_1}
    rotation=<rot -37 -69 2.729>
   }
   Output Pose(Result=Component To Local(ComponentPose=@n38)) {layerGroup=None}
  ]}
 EventGraph {nodes [
   Event BlueprintUpdateAnimation()
   > n1 = Gravity Override State Function DEPRECATED(CharacterPartAnimInstance=Self-Reference(), GameTimeInSeconds=GetGameTimeInSeconds(), GravityOverrideParams=Get Gravity Override Params())
   > Set GravityOverride(GravityOverride=@n1)
   n2 = BreakTransform(InTransform=GetDeltaTransformFromRefPose(self=GetOwnerSkeletalMeshComponent(), BoneName=thigh_l, BaseName=pelvis))
   > Set L thighRotatedBack(L thighRotatedBack:double=MapRangeClamped(Value=BreakRotator(InRot=@n2.Rotation).Yaw, InRangeA=-20, InRangeB=-4)) -> (Output_Get:double)
   n3 = Get bIsKneeSliding()
   > Set Isnt Sliding(Isnt Sliding:double="To Float (Boolean)"(InBool=NOT Boolean(A=@n3))) -> (Output_Get:double)
   > Set Is Sliding(Is Sliding:float="To Float (Boolean)"(InBool=@n3)) -> (Output_Get:float)
   Get bIsSliding()
  ]}
]
variables [
 Gravity Override Params {type=GravityOverrideParamsStruct, default {skydive_DivingDown {gravityOverride=<-1500 0 600>}}}
 GravityOverride {type=Vector}
 Is Sliding {type=number}
 Isnt Sliding {type=number}
 L thighRotatedBack {type=number}
]
